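# Niagara effect for the JoltMosaic back bling, made of short timed bursts of energy, flare, sheen and sparkle.
# An animated material sheen mesh fires first, then a spinning sparkle, then fading energy meshes coincide with two sprite flares split by spawn group.
# Time-of-day inputs dim or tint most layers.
# User variables expose kill count, reactivity, swinging and front-end states; state-change and timeline modules are referenced, but their wiring is not shown.
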NiagaraSystem /BRCosmetics/Accessories/FORT_Backpacks/Backpack_F_MED_JoltMosaic/FX/NS_Backpack_JoltMosaic
dependencies {
 usedDataInterfaces [&NiagaraDataInterfaceVectorCurve, &NiagaraDataInterfaceColorCurve, &NiagaraDataInterfaceCurve]
 usedDynamicInputs [
  &SetBoolByFloatComparison
  &Boolean_AndOperation
  &SetBoolByIntComparison
 ]
 usedModules [
  &self:SystemSpawnScript.SystemScriptSource.SystemScriptGraph.NiagaraNodeAssignment_0.SetVariables_#1
  &SystemState
  /Game/Effects/Niagara/Modules/Logic/UpdateValueStateChange
  &Timeline
 ]
}
emitters {
 NE_Base_Energy {
  emitterUpdateScript [
   &EmitterState()
   &SpawnBurst_Instantaneous(Spawn Count=5, Spawn Time=2.5, Use Spawn Probability=false, UseLoopCountLimit=false)
  ]
  particleSpawnScript [
   &InitializeParticle(
    Color=&FNTimeOfDayColor(Night Color=<rgba 0.1 0.1 0.1 0.4>, Morning Color=<rgba 1 1 1 0.4>, Day Color=<rgba 1 1 1 0.4>, Evening Color=<rgba 1 1 1 0.4>)
    Lifetime=0.35
    Mesh Scale=&RandomRangeVector(Minimum=<0.2 0.3 0.3>, Maximum=<0.2 0.3 0.3>)
    Mesh Scale Mode=::NewEnumerator4
    UsePositionOffset=false
    Write Material Random=true
   )
   &InitialMeshOrientation(Mesh Orientation Mode=::NewEnumerator2, Rotation=<0 0.75 0.125>, Rotation Coordinate Space=::NewEnumerator2, Use Rotation Vector=true)
  ]
  particleUpdateScript [
   &ParticleState()
   &ScaleMeshSize(Scale Factor=&VectorFromCurve(VectorCurve=DataInterface() {propertyValues {
       XCurve {keys [{interpMode=RCIM_Cubic, tangentMode=RCTM_User, time=0, value=0}, {interpMode=RCIM_Cubic, tangentMode=RCTM_User, time=0.5, value=0.8}]}
       YCurve {keys [{arriveTangent=6.5531, interpMode=RCIM_Cubic, leaveTangent=6.5531, tangentMode=RCTM_User, time=0, value=0}, <at 0.5 1>, <at 1 0.2>]}
       ZCurve {keys [{arriveTangent=7.1759, interpMode=RCIM_Cubic, leaveTangent=7.1759, tangentMode=RCTM_User, time=0, value=0}, <at 0.2 1>, <at 1 0.2>]}
      }}))
   &ScaleColor(
    Scale Alpha=&FloatFromCurve(FloatCurve=DataInterface() {propertyValues {Curve {keys [
         {arriveTangent=-2.6258, interpMode=RCIM_Cubic, leaveTangent=-2.6258, tangentMode=RCTM_User, time=0, value=1}
         {interpMode=RCIM_Cubic, tangentMode=RCTM_User, time=1, value=0}
        ]}}})
    Scale RGB=&MakeVectorFromLinearColorRGB(Color=&ColorFromCurve(ColorCurve=DataInterface() {propertyValues {
        AlphaCurve {keys [<at 0 1 RCIM_Cubic>,<at 1 1 RCIM_Cubic>]}
        BlueCurve {keys [
          <at 0.1792 1.5016>
          {arriveTangent=-2.037, interpMode=RCIM_Cubic, leaveTangent=-2.037, time=0.2961, value=1.8924}
          <at 0.8372 0.1612 RCIM_Cubic>
         ]}
        GreenCurve {keys [
          <at 0.1792 6.8223>
          {arriveTangent=-9.8569, interpMode=RCIM_Cubic, leaveTangent=-9.8569, time=0.2961, value=6.1224}
          <at 0.8372 0.3363 RCIM_Cubic>
         ]}
        RedCurve {keys [<at 0.1792 10>, {arriveTangent=-13.6773, interpMode=RCIM_Cubic, leaveTangent=-13.6773, time=0.2961, value=10}, <at 0.8372 1 RCIM_Cubic>]}
       }}))
    ScaleA=true
    ScaleRGB=true
   )
   &CameraOffset(Camera Offset Amount=1)
  ]
  rendererClasses [&NiagaraMeshRendererProperties]
  renderers [{rendererClass=&NiagaraMeshRendererProperties, rendererIndex=0}]
 }
 NE_Flare {
  emitterUpdateScript [
   &EmitterState()
   &SpawnBurst_Instantaneous(Spawn Time=2.5, Use Spawn Probability=false, UseLoopCountLimit=false)
   &SpawnBurst_Instantaneous(Spawn Group=1, Spawn Time=2.5, Use Spawn Probability=false, UseLoopCountLimit=false) {module=SpawnBurst_Instantaneous001}
  ]
  particleSpawnScript [
   /SetVariables(Particles.VisibilityTag=&SpawnGroup()) {isSetParameters=true}
   &InitializeParticle(
    Color Mode=::NewEnumerator0
    Lifetime=0.2
    Sprite Rotation Angle=&MergeHelper_FloatSwitch(0=-45, 1=60, 2=?, 3=?, Index=$Particles.VisibilityTag, Num=::NewEnumerator0)
    Sprite Rotation Mode=::NewEnumerator0
    Sprite Size Mode=::NewEnumerator3
    Uniform Sprite Size=&MergeHelper_FloatSwitch(0=80, 1=35, 2=?, 3=?, Index=$Particles.VisibilityTag, Num=::NewEnumerator0)
    UsePositionOffset=false
    Write Material Random=true
   )
  ]
  particleUpdateScript [
   &ParticleState()
   &ScaleSpriteSize()
   &CameraOffset(Camera Offset Amount=&MergeHelper_FloatSwitch(0=10, 1=10, 2=?, 3=?, Index=$Particles.VisibilityTag, Num=::NewEnumerator0))
   &ScaleColor(
    Scale Alpha=&MergeHelper_FloatSwitch(
     0=&FNTimeOfDayScalar(Night Scale=0.5, Morning Scale=1, Evening Scale=1)
     1=&FNTimeOfDayScalar(Night Scale=0.25, Morning Scale=0.5, Day Scale=0.5, Evening Scale=0.5)
     2=?
     3=?
     Index=$Particles.VisibilityTag
     Num=::NewEnumerator0
    )
    ScaleA=true
    ScaleRGB=true
   )
  ]
  rendererClasses [&NiagaraSpriteRendererProperties]
  renderers [
   {rendererClass=&NiagaraSpriteRendererProperties}
   {rendererClass=&NiagaraSpriteRendererProperties, rendererIndex=1}
  ]
 }
 NE_Sheen_Mesh {
  emitterUpdateScript [
   &EmitterState()
   &SpawnBurst_Instantaneous(Use Spawn Probability=false, UseLoopCountLimit=false)
  ]
  particleSpawnScript [
   &InitializeParticle(
    Color=<rgba 100 100 100 1>
    Lifetime=0.5
    Mesh Scale=<0.065 0.05 0.04>
    Mesh Scale Mode=::NewEnumerator4
    Position Offset=<-0.25 -1 5>
    Position Offset Coordinate Space=::Local
    UsePositionOffset=true
    Write Material Random=true
   )
   &DynamicMaterialParameters(
    Index 0 Param 2=0.75
    Module.Index0Float0Write=true
    Module.Index0Float1Write=true
    Module.Index0Float2Write=true
    Module.Index0Float3Write=true
   )
   &InitialMeshOrientation(Mesh Orientation Mode=::NewEnumerator2, Orientation Vector=<0 1 0>, Rotation Coordinate Space=::NewEnumerator2, Use Rotation Vector=false)
  ]
  particleUpdateScript [
   &ParticleState()
   &DynamicMaterialParameters(
    Index 0 Param 1=&FloatFromCurve(FloatCurve=DataInterface() {propertyValues {Curve {keys [{interpMode=RCIM_Cubic, tangentMode=RCTM_User, time=0, value=0}, {interpMode=RCIM_Cubic, tangentMode=RCTM_User, time=1, value=1}]}}})
    Index 0 Param 2=1
    Module.Index0Float0Write=true
    Module.Index0Float1Write=true
    Module.Index0Float2Write=true
    Module.Index0Float3Write=true
   ) {module=DynamicMaterialParameters001}
   &ScaleColor(
    Color Value To Scale=$Particles.Color
    Scale Alpha=&FNTimeOfDayScalar(Night Scale=0.9, Morning Scale=0.92, Day Scale=0.95, Evening Scale=0.92)
    Scale RGB=&VectorFromCurve(VectorCurve=DataInterface() {propertyValues {XCurve {keys [<at 0 1>,<at 0.5 1>,<at 1 1>]}, YCurve {keys [<at 0 1>,<at 0.5 1>,<at 1 1>]}, ZCurve {keys [<at 0 1>,<at 0.5 1>,<at 1 0>]}}})
    ScaleA=true
    ScaleRGB=true
   )
  ]
  rendererClasses [&NiagaraMeshRendererProperties]
  renderers [{rendererClass=&NiagaraMeshRendererProperties, rendererIndex=0}]
 }
 NE_Sparkles {
  emitterUpdateScript [
   &EmitterState()
   &SpawnBurst_Instantaneous(Spawn Time=0.2, Use Spawn Probability=false, UseLoopCountLimit=false)
  ]
  particleSpawnScript [
   &InitializeParticle(
    Color=<rgba 2 2 2 1>
    Lifetime=0.5
    Position Offset=<0 -1.2 10>
    Position Offset Coordinate Space=::Local
    Sprite Rotation Mode=::NewEnumerator2
    Sprite Size Mode=::NewEnumerator3
    Uniform Sprite Size=200
    UsePositionOffset=true
    Write Material Random=true
   )
   &CameraOffset(Camera Offset Amount=5)
  ]
  particleUpdateScript [
   &ParticleState()
   &ScaleSpriteSize(
    Uniform Curve Scale=0.2
    Uniform Curve Sprite Scale=DataInterface() {propertyValues {Curve {keys [
        {arriveTangent=9.0725, interpMode=RCIM_Cubic, leaveTangent=9.0725, tangentMode=RCTM_User, time=0, value=0}
        {arriveTangent=-1.7917, interpMode=RCIM_Cubic, leaveTangent=-1.7917, tangentMode=RCTM_User, time=0.0666, value=0.9847}
        {arriveTangent=-0.0405, interpMode=RCIM_Cubic, leaveTangent=-0.0405, tangentMode=RCTM_User, time=1, value=0}
       ]}}}
   )
   &SpriteRotationRate(Rotation Rate=200)
  ]
  rendererClasses [&NiagaraSpriteRendererProperties]
  renderers [{rendererClass=&NiagaraSpriteRendererProperties}]
 }
}
summary {
 systemName=NS_Backpack_JoltMosaic
 emitters [
  {emitterName=NE_Base_Energy, rendererClasses [&NiagaraMeshRendererProperties]}
  {emitterName=NE_Flare, rendererClasses [&NiagaraSpriteRendererProperties]}
  {emitterName=NE_Sheen_Mesh, rendererClasses [&NiagaraMeshRendererProperties]}
  {emitterName=NE_Sparkles, rendererClasses [&NiagaraSpriteRendererProperties]}
 ]
 userVariables [
  User.bIsFrontEnd {type=NiagaraBool}
  User.bIsFrontEndPreview {type=NiagaraBool}
  User.bisSwinging {type=NiagaraBool}
  User.IsReactive {type=NiagaraFloat, defaultValue=1, description=""}
  User.KillCount {type=NiagaraInt32}
 ]
}
