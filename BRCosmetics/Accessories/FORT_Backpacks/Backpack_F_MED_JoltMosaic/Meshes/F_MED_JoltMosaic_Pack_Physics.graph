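# Physics asset for the JoltMosaic backpack cosmetic, giving bodies and joints to its dangling flower and star chains and its handle.
# Each chain starts at a base capsule rotationally locked to the pack capsule, then runs through three flower or five star capsule links.
# Chain joints use tight angular limits, with a wider swing at each first link.
# The handle is a box jointed to the pack with free swing and twist.
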PhysicsAsset /BRCosmetics/Accessories/FORT_Backpacks/Backpack_F_MED_JoltMosaic/Meshes/F_MED_JoltMosaic_Pack_Physics
bodies {
 base_dyn_flowerChain [{shapeType=Capsule, length=0.505, radius=0.505, shapeName=base_dyn_flowerChain_capsule, rotation=<rot 0 -90 0>}]
 base_dyn_handle [{
   shapeType=Box
   extentX=8.2224
   extentY=8.6993
   extentZ=1.5
   shapeName=base_dyn_handle_box
   center=<-0.424 0.0073 0.1091>
   rotation=<rot 0 89.4929 -89.1107>
  }]
 base_dyn_starChain [{shapeType=Capsule, length=0.505, radius=0.505, shapeName=base_dyn_starChain_capsule, rotation=<rot 0 -90 0>}]
 dyn_flowerChain_00 [{
   shapeType=Capsule
   length=1.2775
   radius=0.6899
   shapeName=dyn_flowerChain_00_capsule
   center=<0.7233 0 0>
   rotation=<rot 0 89.9925 -89.9822>
  }]
 dyn_flowerChain_01 [{
   shapeType=Capsule
   length=1.3666
   radius=0.6993
   shapeName=dyn_flowerChain_01_capsule
   center=<0.8124 0 0>
   rotation=<rot 0 90.0999 -89.1467>
  }]
 dyn_flowerChain_02 [{
   shapeType=Capsule
   length=1.4316
   radius=0.6888
   shapeName=dyn_flowerChain_02_capsule
   center=<0.9553 0 0>
   rotation=<rot 0 90.2287 -87.7476>
  }]
 dyn_starChain_00 [{shapeType=Capsule, length=1.2058, radius=0.6746, shapeName=dyn_starChain_00_capsule, center=<0.3443 0 0>, rotation=<rot 0 89.895 -89.9948>}]
 dyn_starChain_01 [{shapeType=Capsule, length=1.7685, radius=0.6746, shapeName=dyn_starChain_01_capsule, center=<1.0404 0 0>, rotation=<rot 0 89.9509 -89.9976>}]
 dyn_starChain_02 [{shapeType=Capsule, length=1.7685, radius=0.6746, shapeName=dyn_starChain_02_capsule, center=<1.0404 0 0>, rotation=<rot 0 89.9509 -89.9976>}]
 dyn_starChain_03 [{shapeType=Capsule, length=1.7685, radius=0.6746, shapeName=dyn_starChain_03_capsule, center=<1.0537 0 0>, rotation=<rot 0 89.9509 -89.9976>}]
 dyn_starChain_04 [{shapeType=Capsule, length=1.7685, radius=0.6746, shapeName=dyn_starChain_04_capsule, center=<1.0454 0 0>, rotation=<rot 0 89.9509 -89.9976>}]
 pack [{shapeType=Capsule, length=25.3597, radius=1.3361, center=<-2.6524 -0.7434 -10.2853>, rotation=<rot -10.5741 -93.8423 49.6933>}]
}
constraints [
 {
  bone1Name=base_dyn_flowerChain
  bone2Name=pack
  swing1LimitDegrees=45
  swing1Motion=Locked
  swing2LimitDegrees=45
  swing2Motion=Locked
  twistLimitDegrees=45
  twistMotion=Locked
 }
 {bone1Name=dyn_flowerChain_00, bone2Name=base_dyn_flowerChain, swing1LimitDegrees=15, swing2LimitDegrees=45, twistLimitDegrees=10}
 {bone1Name=dyn_flowerChain_01, bone2Name=dyn_flowerChain_00, swing1LimitDegrees=10, swing2LimitDegrees=10, twistLimitDegrees=10}
 {bone1Name=dyn_flowerChain_02, bone2Name=dyn_flowerChain_01, swing1LimitDegrees=10, swing2LimitDegrees=10, twistLimitDegrees=10}
 {
  bone1Name=base_dyn_starChain
  bone2Name=pack
  swing1LimitDegrees=45
  swing1Motion=Locked
  swing2LimitDegrees=45
  swing2Motion=Locked
  twistLimitDegrees=45
  twistMotion=Locked
 }
 {bone1Name=dyn_starChain_00, bone2Name=base_dyn_starChain, swing1LimitDegrees=15, swing2LimitDegrees=45, twistLimitDegrees=10}
 {bone1Name=dyn_starChain_01, bone2Name=dyn_starChain_00, swing1LimitDegrees=10, swing2LimitDegrees=10, twistLimitDegrees=10}
 {bone1Name=dyn_starChain_02, bone2Name=dyn_starChain_01, swing1LimitDegrees=10, swing2LimitDegrees=10, twistLimitDegrees=10}
 {bone1Name=dyn_starChain_03, bone2Name=dyn_starChain_02, swing1LimitDegrees=10, swing2LimitDegrees=10, twistLimitDegrees=10}
 {bone1Name=dyn_starChain_04, bone2Name=dyn_starChain_03, swing1LimitDegrees=10, swing2LimitDegrees=10, twistLimitDegrees=10}
 {
  bone1Name=base_dyn_handle
  bone2Name=pack
  swing1LimitDegrees=45
  swing1Motion=Free
  swing2LimitDegrees=45
  swing2Motion=Free
  twistLimitDegrees=45
  twistMotion=Free
 }
]
