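# Animation Blueprint for the JoltMosaic backpack cosmetic, adding secondary motion to the pose copied from its parent mesh.
# When rigid body or cloth is enabled, a RigidBody simulation in pack bone space drives it; otherwise trail controllers swing the flower and star chains.
# A final offset shifts the pack bone.
# Events set gravity override for the simulation, plus skydive, elimination and sim-space variables that the AnimGraph never reads.
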AnimBlueprint /BRCosmetics/Accessories/FORT_Backpacks/Backpack_F_MED_JoltMosaic/Meshes/F_MED_JoltMosaic_Pack_AnimBP
target_skeleton=&/F_MED_JoltMosaic_Pack_Skeleton
graphs [
 AnimGraph {nodes [
   "Transform (Modify) Bone - Bone: base_dyn_handle"(Alpha=1, Translation="9,-0.133,-0.5", Rotation="0,0,0", Scale="1,1,1") {
    rotationMode=BMM_Additive
    rotationSpace=BCS_BoneSpace
    translationMode=BMM_Additive
    translationSpace=BCS_BoneSpace
    alphaBoolBlend {blendOption=Linear}
    alphaScaleBiasClamp {bInterpResult=true, interpSpeedDecreasing=15, interpSpeedIncreasing=25}
    boneToModify {boneName=base_dyn_handle}
    translation=<9 -0.133 -0.5>
   }
   n1 = Local To Component(LocalPose=Blend Poses by bool(
     BlendPose_0=Use cached pose 'Rban_Cache'()
     BlendPose_1=Use cached pose 'Trail_Cache'()
     bActiveValue=Get RbanOrCloth() -> (RbanOrCloth:bool)
    ) {blendType=Linear})
   Output Pose(Result=Component To Local(ComponentPose="Transform (Modify) Bone - Bone: pack"(ComponentPose=@n1, Translation="0,-2,0", Rotation="0,0,0") {translationMode=BMM_Additive, translationSpace=BCS_BoneSpace, boneToModify {boneName=pack}, translation {y=-2}})) {layerGroup=None}
   Save cached pose 'Start_Cache'(Pose=Copy Pose From Mesh()) {cacheName=Start_Cache}
   n2 = RigidBody!(
    ComponentPose=Local To Component(LocalPose=Use cached pose 'Start_Cache'())
    Alpha
    OverrideWorldGravity:Vector=Get GravityOverride()
    ExternalForce
    SimSpaceSettings:Sim Space Settings Structure=(WorldAlpha=0.2,VelocityScaleZ=1,DampingAlpha=1,MaxLinearVelocity=10000,MaxAngularVelocity=10000,MaxLinearAcceleration=10000,MaxAngularAcceleration=10000,ExternalLinearDragV=(X=0,Y=0,Z=0),ExternalLinearVelocity=(X=0,Y=0,Z=0),ExternalAngularVelocity=(X=0,Y=0,Z=0))
   ) -> (Pose) {
    bOverrideWorldGravity=true
    overridePhysicsAsset=&/F_MED_JoltMosaic_Pack_Physics
    simulationSpace=BaseBoneSpace
    alphaBoolBlend {blendOption=Linear}
    baseBoneRef {boneName=pack}
    componentAppliedLinearAccClamp=<3000 3600 5000>
    componentLinearAccScale=<0.3 0.36 0.5>
    componentLinearVelScale=<0.3 0.36 0.5>
    simSpaceSettings {worldAlpha=0.2}
   }
   Save cached pose 'Rban_Cache'(Pose=Component To Local(ComponentPose=@n2)) {cacheName=Rban_Cache}
   n3 = "Transform (Modify) Bone - Bone: base_dyn_flowerChain"(
    ComponentPose=Local To Component(LocalPose=Use cached pose 'Start_Cache'())
    Alpha=1
    Translation="0,0,0"
    Rotation="-45,15,0"
    Scale="1,1,1"
   ) {
    rotationMode=BMM_Additive
    rotationSpace=BCS_BoneSpace
    scaleSpace=BCS_BoneSpace
    translationSpace=BCS_BoneSpace
    boneToModify {boneName=base_dyn_flowerChain}
    rotation {pitch=-45, yaw=15}
   }
   n4 = "Transform (Modify) Bone - Bone: base_dyn_starChain"(ComponentPose=@n3, Alpha=1, Translation="0,0,0", Rotation="-45,15,0", Scale="1,1,1") {
    rotationMode=BMM_Additive
    rotationSpace=BCS_BoneSpace
    translationSpace=BCS_BoneSpace
    boneToModify {boneName=base_dyn_starChain}
    rotation {pitch=-45, yaw=15}
   }
   n5 = Trail controller - Bone: dyn_flowerChain_02(ComponentPose=@n4, Alpha=1) {
    bLimitRotation=true
    bUsePlanarLimit=true
    chainLength=3
    debugLifeTime=0.1
    maxDeltaTime=0.0333
    relaxationSpeedScale=2.5
    alphaScaleBiasClamp {bInterpResult=true}
    baseJoint {boneName=root}
    rotationLimits [{limitMax=<5 5 0>, limitMin=<-5 -5 -5>}, {limitMax=<5 5 0>, limitMin=<-5 -5 -5>}, {limitMax=<5 5 0>, limitMin=<-5 -5 -5>}]
    rotationOffsets [<0 0 0>, <0 0 0>, <0 0 0>]
    trailBone {boneName=dyn_flowerChain_02}
    trailRelaxationSpeed {editorCurveData {keys [<at 0 10 RCIM_Cubic>,<at 1 5 RCIM_Cubic>]}}
    alphaBoolBlend=~
   }
   n6 = Trail controller - Bone: dyn_starChain_04(ComponentPose=@n5, Alpha=1) {
    bLimitRotation=true
    bUsePlanarLimit=true
    chainLength=5
    debugLifeTime=0.1
    maxDeltaTime=0.0333
    relaxationSpeedScale=2.5
    alphaScaleBiasClamp {bInterpResult=true}
    baseJoint {boneName=root}
    rotationLimits [
     {limitMax=<5 5 5>, limitMin=<-5 -5 -5>}
     {limitMax=<5 5 5>, limitMin=<-5 -5 -5>}
     {limitMax=<5 5 5>, limitMin=<-5 -5 -5>}
     {limitMax=<5 5 5>, limitMin=<-5 -5 -5>}
     {limitMax=<5 5 5>, limitMin=<-5 -5 -5>}
    ]
    rotationOffsets [<0 0 0>, <0 0 0>, <0 0 0>, <0 0 0>, <0 0 0>]
    trailBone {boneName=dyn_starChain_04}
    trailRelaxationSpeed {editorCurveData {keys [<at 0 10 RCIM_Cubic>,<at 1 5 RCIM_Cubic>]}}
    alphaBoolBlend=~
   }
   Save cached pose 'Trail_Cache'(Pose=Component To Local(ComponentPose=@n6)) {cacheName=Trail_Cache}
  ]}
 EventGraph {nodes [
   n1 = Event BlueprintUpdateAnimation()
   > n2 = Sequence!(execute) -> (then_0, then_1, then_2:Exec)
   > n3 = Gravity Override State Function DEPRECATED(CharacterPartAnimInstance=Self-Reference(), GameTimeInSeconds=@n1.DeltaTimeX, GravityOverrideParams=Get Gravity Override Params())
   > Set GravityOverride(GravityOverride=@n3)
   Set RbanOrCloth(execute=@n2.then_1, RbanOrCloth:bool=OR Boolean(A=Get bRigidBodyEnabled(), B=Get bClothEnabled())) -> (Output_Get:bool)
   n4 = Event BlueprintProcessAnimInputEvent()
   > Branch(Condition=IsNamed(self=@n4.AnimInputEvent, InEventName=aie.EnemyPlayerEliminated))
   > Set isEnemyPlayerEliminated(isEnemyPlayerEliminated:double=1) -> (Output_Get:double)
   > Delay(Duration=2.5)
   > Set isEnemyPlayerEliminated(isEnemyPlayerEliminated=0)
   Event OnBeginSkydiving()
   > Set isSkydive(isSkydive:double=1) -> (Output_Get:double)
   Event OnEndSkydiving()
   > Set isSkydive(isSkydive=0)
   Event OnCharacterCustomizationLoaded()
   > n5 = Branch(Condition=Get bIsInFrontEnd())
   > Set Sim Space Settings(Sim Space Settings:Sim Space Settings Structure=Get Sim Space Settings Front End() -> (Sim Space Settings Front End:Sim Space Settings Structure)) -> (Output_Get:Sim Space Settings Structure)
   > n6 = Delay(Duration=2)
   Set isEnemyPlayerEliminated(execute=[@n6.then, @n7.then], isEnemyPlayerEliminated=1)
   > Delay(Duration=2.5)
   > Set isEnemyPlayerEliminated(isEnemyPlayerEliminated=0)
   > n7 = Delay(Duration=15)
   Set Sim Space Settings(execute=@n5.else, Sim Space Settings=Get Sim Space Settings in Zone() -> (Sim Space Settings in Zone:Sim Space Settings Structure))
   TryGetPawnOwner()
   Event BlueprintBeginPlay() {is=title K2Node_Event}
  ]}
]
variables [
 Gravity Override Params {type=GravityOverrideParamsStruct}
 GravityOverride {type=Vector}
 isEnemyPlayerEliminated {type=number}
 isSkydive {type=number}
 RbanOrCloth {type=boolean}
 Sim Space Settings {
  type=SimSpaceSettings
  default {
   dampingAlpha=1
   maxAngularAcceleration=10000
   maxAngularVelocity=10000
   maxLinearAcceleration=10000
   maxLinearVelocity=10000
   velocityScaleZ=1
   worldAlpha=0
   externalLinearDragV=<0 0 0>
  }
 }
 Sim Space Settings Front End {
  type=SimSpaceSettings
  default {
   dampingAlpha=1
   maxAngularAcceleration=10000
   maxAngularVelocity=10000
   maxLinearAcceleration=10000
   maxLinearVelocity=10000
   velocityScaleZ=1
   worldAlpha=0.38
   externalLinearDragV=<0 0 0>
  }
 }
 Sim Space Settings in Zone {
  type=SimSpaceSettings
  default {
   dampingAlpha=1
   maxAngularAcceleration=10000
   maxAngularVelocity=10000
   maxLinearAcceleration=10000
   maxLinearVelocity=10000
   velocityScaleZ=1
   worldAlpha=0.2
   externalLinearDragV=<0 0 0>
  }
 }
]
